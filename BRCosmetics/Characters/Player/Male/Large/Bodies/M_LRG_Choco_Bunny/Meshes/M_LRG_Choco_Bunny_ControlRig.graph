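# Control Rig for the large male Choco Bunny player body cosmetic that drives the dyn_jaw bone from C_jaw.
# It reads C_jaw's local rotation as Euler angles, triples X, keeps Y and Z, and rebuilds the quaternion.
# The result overrides dyn_jaw's whole local transform, with zero translation and unit scale.
# The other dyn_ bones, such as gloves, collar and straps, are left untouched by the graph.
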ControlRigBlueprint /BRCosmetics/Characters/Player/Male/Large/Bodies/M_LRG_Choco_Bunny/Meshes/M_LRG_Choco_Bunny_ControlRig
graphs [
 RigVMFunctionLibrary {links [], nodes []}
 RigVMModel {nodes [
   n1 = &RigUnit_BeginExecution()
   n2 = &RigVMFunction_MathQuaternionToEuler(RotationOrder=ZYX, Value=&RigUnit_GetBoneTransform(Bone=C_jaw, Space=LocalSpace).Transform.Rotation)
   /Script/ControlRig.RigUnit_ModifyBoneTransforms(
    BoneToModify=((Bone="dyn_jaw",Transform=(Rotation=(X=0,Y=0,Z=0,W=1),Translation=(X=0,Y=0,Z=0),Scale3D=(X=1,Y=1,Z=1))))
    Mode=OverrideLocal
    Weight=1
    WeightMaximum=1
    WeightMinimum=0
    ExecutePin=@n1.ExecutePin
    BoneToModify.0.Transform.Rotation=&RigVMFunction_MathQuaternionFromEuler(
     Euler=(X=0,Y=0,Z=0)
     Euler.Y=@n2.Result.Y
     Euler.Z=@n2.Result.Z
     Euler.X=/Script/RigVM.RigVMFunction_MathVectorScale(Factor=3, Value=(X=0,Y=0,Z=0), Value.X=@n2.Result.X) {is=struct RigVMUnitNode, template="Scale::Execute(in Value,in Factor,out Result)"}.Result.X
    ).Result
   ) {is=struct RigVMUnitNode}
  ]}
]
hierarchy [
 root {loc=<0 0 0>, rot=<0 0 0 1>}
 pelvis {parent=root, loc=<0 -0.7321 92.4218>, rot=<0.0127 0.707 -0.0127 -0.707>}
 spine_01 {parent=pelvis, loc=<5.6002 -0.5128 0>, rot=<0 0 0.034 -0.9994>}
 spine_02 {parent=spine_01, loc=<8.0753 0.0139 0>, rot=<0 0 0.0238 -0.9997>}
 spine_03 {parent=spine_02, loc=<7.0393 0 0>, rot=<0 0 0.0149 -0.9999>}
 spine_04 {parent=spine_03, loc=<8.1024 0 0>, rot=<0 0 -0.0132 -0.9999>}
 spine_05 {parent=spine_04, loc=<10.0146 0 0>, rot=<0 0 -0.0015 -1>}
 clavicle_l {parent=spine_05, loc=<2.5856 -0.2866 -4.3777>, rot=<0.0465 -0.7289 -0.009 -0.6829>}
 upperarm_l {parent=clavicle_l, loc=<18.2264 0.1342 -0.001>, rot=<0.0253 -0.2693 0.0013 -0.9627>}
 lowerarm_l {parent=upperarm_l, loc=<29.0071 0.1686 -1.239>, rot=<0.0391 -0.0099 -0.2453 -0.9686>}
 hand_l {parent=lowerarm_l, loc=<25.8607 0.222 1.0274>, rot=<-0.7193 -0.0988 0.1193 -0.6773>}
 index_metacarpal_l {parent=hand_l, loc=<1.6535 0.0183 -2.96>, rot=<0.0158 -0.0124 -0.0173 -0.9996>}
 index_01_l {parent=index_metacarpal_l, loc=<8.3008 0.0357 -0.0524>, rot=<-0.0044 0.0485 0.0924 -0.9945>}
 index_02_l {parent=index_01_l, loc=<3.3587 -0.3003 0.0452>, rot=<-0.002 0.0027 0.3225 -0.9466>}
 index_03_l {parent=index_02_l, loc=<2.5578 -0.1231 -0.0252>, rot=<-0.002 0.0067 0.2558 -0.9667>}
 middle_metacarpal_l {parent=hand_l, loc=<2.1856 -0.1725 -0.5984>, rot=<0.006 0.0168 -0.0329 -0.9993>}
 middle_01_l {parent=middle_metacarpal_l, loc=<7.3275 -0.2447 0.0129>, rot=<-0.0208 0.0042 0.0858 -0.9961>}
 middle_02_l {parent=middle_01_l, loc=<4.0123 0.1401 0.1175>, rot=<-0.0218 0.0258 0.3512 -0.9357>}
 middle_03_l {parent=middle_02_l, loc=<2.6077 0.0534 0.1373>, rot=<0.0082 0.0188 0.1708 -0.9851>}
 pinky_metacarpal_l {parent=hand_l, loc=<1.7616 -0.9753 3.3814>, rot=<-0.0397 0.0814 -0.069 -0.9935>}
 pinky_01_l {parent=pinky_metacarpal_l, loc=<7.1709 -0.4831 -0.1094>, rot=<-0.0654 -0.0035 0.2242 -0.9723>}
 pinky_02_l {parent=pinky_01_l, loc=<2.8274 0.0892 -0.1043>, rot=<0.0037 0.0011 0.2854 -0.9584>}
 pinky_03_l {parent=pinky_02_l, loc=<2.2046 0.0307 -0.0078>, rot=<-0.0002 0.0098 0.1445 -0.9895>}
 ring_metacarpal_l {parent=hand_l, loc=<2.2074 -0.3625 1.3638>, rot=<-0.0284 0.0722 -0.0458 -0.9959>}
 ring_01_l {parent=ring_metacarpal_l, loc=<7.0295 -0.6197 -0.1446>, rot=<-0.0489 -0.0252 0.1892 -0.9804>}
 ring_02_l {parent=ring_01_l, loc=<3.6174 0.4647 -0.1027>, rot=<0 0 0.2806 -0.9598>}
 ring_03_l {parent=ring_02_l, loc=<2.4152 0.1141 0.0199>, rot=<0 0 0.1189 -0.9929>}
 thumb_01_l {parent=hand_l, loc=<1.3118 -0.3191 -2.459>, rot=<0.6546 -0.4042 0.0664 -0.6354>}
 thumb_02_l {parent=thumb_01_l, loc=<5.9336 -0.0448 0.0401>, rot=<-0.0037 0.0213 0.298 -0.9543>}
 thumb_03_l {parent=thumb_02_l, loc=<2.2223 0.3682 -0.0907>, rot=<0.0118 -0.013 0.0066 -0.9998>}
 lowerarm_twist_01_l {parent=lowerarm_l, loc=<17.8549 0 0>, rot=<0 0 0 1>}
 dyn_glove_b_base_l {parent=lowerarm_twist_01_l, loc=<7.9473 -3.7744 1.3142>, rot=<0.2283 -0.2772 -0.2106 -0.9092>}
 dyn_glove_b_l {parent=dyn_glove_b_base_l, loc=<0 0 0>, rot=<0 0 0 1>}
 dyn_glove_c_base_l {parent=lowerarm_twist_01_l, loc=<7.061 -4.1642 -0.5878>, rot=<-0.2296 -0.1308 -0.1241 -0.9564>}
 dyn_glove_c_l {parent=dyn_glove_c_base_l, loc=<0 0 0>, rot=<0 0 0 -1>}
 dyn_glove_d_base_l {parent=lowerarm_twist_01_l, loc=<6.5435 -0.524 -3.0256>, rot=<-0.5986 0.0717 -0.1494 -0.7837>}
 dyn_glove_d_l {parent=dyn_glove_d_base_l, loc=<0 0 0>, rot=<0 0 0 1>}
 dyn_glove_e_base_l {parent=lowerarm_twist_01_l, loc=<6.0598 3.0509 -2.778>, rot=<-0.8475 -0.0116 0.045 -0.5288>}
 dyn_glove_e_l {parent=dyn_glove_e_base_l, loc=<0 0 0>, rot=<0 0 0 -1>}
 dyn_glove_a_base_l {parent=lowerarm_twist_01_l, loc=<9.4781 1.9164 3.1425>, rot=<0.7422 -0.1986 -0.1595 -0.6199>}
 dyn_glove_a_l {parent=dyn_glove_a_base_l, loc=<0 0 0>, rot=<0 0 0 -1>}
 lowerarm_twist_02_l {parent=lowerarm_l, loc=<9.6199 0 0>, rot=<0 0 0 1>}
 upperarm_twist_01_l {parent=upperarm_l, loc=<7.9796 0.0843 -0.6195>, rot=<0 0 0 1>}
 upperarm_twist_02_l {parent=upperarm_l, loc=<19.5059 0.0843 -0.6195>, rot=<0 0 0 1>}
 clavicle_r {parent=spine_05, loc=<2.5856 -0.2865 4.3777>, rot=<0.7289 0.0465 -0.6829 0.009>}
 upperarm_r {parent=clavicle_r, loc=<-18.2264 -0.1342 0.001>, rot=<0.0253 -0.2693 0.0013 -0.9627>}
 lowerarm_r {parent=upperarm_r, loc=<-29.0072 -0.1686 1.239>, rot=<0.0391 -0.0099 -0.2453 -0.9686>}
 hand_r {parent=lowerarm_r, loc=<-25.8608 -0.222 -1.0274>, rot=<-0.7193 -0.0988 0.1193 -0.6773>}
 index_metacarpal_r {parent=hand_r, loc=<-1.6534 -0.0181 2.9601>, rot=<0.0158 -0.0124 -0.0173 -0.9996>}
 index_01_r {parent=index_metacarpal_r, loc=<-8.3008 -0.0357 0.0524>, rot=<-0.0044 0.0485 0.0924 -0.9945>}
 index_02_r {parent=index_01_r, loc=<-3.3589 0.3003 -0.0453>, rot=<-0.002 0.0027 0.3225 -0.9466>}
 index_03_r {parent=index_02_r, loc=<-2.5578 0.123 0.0252>, rot=<-0.002 0.0067 0.2558 -0.9667>}
 middle_metacarpal_r {parent=hand_r, loc=<-2.1856 0.1726 0.5984>, rot=<0.006 0.0168 -0.0329 -0.9993>}
 middle_01_r {parent=middle_metacarpal_r, loc=<-7.3273 0.2447 -0.0129>, rot=<-0.0208 0.0042 0.0858 -0.9961>}
 middle_02_r {parent=middle_01_r, loc=<-4.0124 -0.14 -0.1175>, rot=<-0.0218 0.0258 0.3512 -0.9357>}
 middle_03_r {parent=middle_02_r, loc=<-2.6077 -0.0533 -0.1373>, rot=<0.0082 0.0188 0.1708 -0.9851>}
 pinky_metacarpal_r {parent=hand_r, loc=<-1.7617 0.9753 -3.3814>, rot=<-0.0397 0.0814 -0.069 -0.9935>}
 pinky_01_r {parent=pinky_metacarpal_r, loc=<-7.1707 0.4832 0.1094>, rot=<-0.0654 -0.0035 0.2242 -0.9723>}
 pinky_02_r {parent=pinky_01_r, loc=<-2.8274 -0.0892 0.1043>, rot=<0.0037 0.0011 0.2854 -0.9584>}
 pinky_03_r {parent=pinky_02_r, loc=<-2.2045 -0.031 0.0078>, rot=<-0.0002 0.0098 0.1445 -0.9895>}
 ring_metacarpal_r {parent=hand_r, loc=<-2.2073 0.3627 -1.3638>, rot=<-0.0284 0.0722 -0.0458 -0.9959>}
 ring_01_r {parent=ring_metacarpal_r, loc=<-7.0297 0.6197 0.1446>, rot=<-0.0489 -0.0252 0.1892 -0.9804>}
 ring_02_r {parent=ring_01_r, loc=<-3.6172 -0.4648 0.1027>, rot=<0 0 0.2806 -0.9598>}
 ring_03_r {parent=ring_02_r, loc=<-2.4153 -0.1141 -0.0199>, rot=<0 0 0.1189 -0.9929>}
 thumb_01_r {parent=hand_r, loc=<-1.3119 0.3192 2.459>, rot=<0.6546 -0.4042 0.0664 -0.6354>}
 thumb_02_r {parent=thumb_01_r, loc=<-5.9336 0.0448 -0.0401>, rot=<-0.0037 0.0213 0.298 -0.9543>}
 thumb_03_r {parent=thumb_02_r, loc=<-2.2223 -0.3682 0.0907>, rot=<0.0118 -0.013 0.0066 -0.9998>}
 lowerarm_twist_01_r {parent=lowerarm_r, loc=<-17.8549 0 0>, rot=<0 0 0 1>}
 dyn_glove_b_base_r {parent=lowerarm_twist_01_r, loc=<-7.9474 3.7745 -1.3138>, rot=<0.2283 -0.2772 -0.2106 -0.9092>}
 dyn_glove_b_r {parent=dyn_glove_b_base_r, loc=<0 0 0>, rot=<0 0 0 1>}
 dyn_glove_c_base_r {parent=lowerarm_twist_01_r, loc=<-7.0609 4.1642 0.5879>, rot=<-0.2296 -0.1308 -0.1241 -0.9564>}
 dyn_glove_c_r {parent=dyn_glove_c_base_r, loc=<0 0 0>, rot=<0 0 0 1>}
 dyn_glove_d_base_r {parent=lowerarm_twist_01_r, loc=<-6.5434 0.524 3.0257>, rot=<-0.5986 0.0717 -0.1494 -0.7837>}
 dyn_glove_d_r {parent=dyn_glove_d_base_r, loc=<0 0 0>, rot=<0 0 0 -1>}
 dyn_glove_e_base_r {parent=lowerarm_twist_01_r, loc=<-6.0597 -3.0509 2.7781>, rot=<-0.8475 -0.0116 0.045 -0.5288>}
 dyn_glove_e_r {parent=dyn_glove_e_base_r, loc=<0 0 0>, rot=<0 0 0 -1>}
 dyn_glove_a_base_r {parent=lowerarm_twist_01_r, loc=<-9.4782 -1.9163 -3.1421>, rot=<0.7422 -0.1986 -0.1595 -0.6199>}
 dyn_glove_a_r {parent=dyn_glove_a_base_r, loc=<0 0 0>, rot=<0 0 0 -1>}
 lowerarm_twist_02_r {parent=lowerarm_r, loc=<-9.6199 0 0>, rot=<0 0 0 1>}
 upperarm_twist_01_r {parent=upperarm_r, loc=<-7.9796 -0.0843 0.6195>, rot=<0 0 0 -1>}
 upperarm_twist_02_r {parent=upperarm_r, loc=<-19.5059 -0.0843 0.6195>, rot=<0 0 0 -1>}
 neck_01 {parent=spine_05, loc=<10.815 1.9196 0>, rot=<0 0 -0.1864 -0.9825>}
 neck_02 {parent=neck_01, loc=<4.7347 0.3045 0>, rot=<0 0 0.0363 -0.9993>}
 head {parent=neck_02, loc=<5.7632 0.5552 0>, rot=<0 0 0.115 -0.9934>}
 faceAttach {parent=head, loc=<-0.8047 -2.8259 0>, rot=<0.0044 -0.3827 -0.0044 -0.9239>, scale=<0.89 0.89 0.89>}
 C_jaw {parent=faceAttach, loc=<0.3124 3.2483 0.3583>, rot=<0.0044 -0.3827 -0.0044 -0.9239>}
 tongue {parent=C_jaw, loc=<0.0178 1.5245 -1.0399>, rot=<-0.1568 0 0 -0.9876>}
 dyn_jaw {parent=faceAttach, loc=<-1.5052 4.5006 -1.4748>, rot=<0.0044 -0.3827 -0.0044 -0.9239>}
 dyn_collar_fr_base_l {parent=neck_01, loc=<-0.6043 7.531 -14.0471>, rot=<-0.0043 -0.986 -0.1665 -0.0007>}
 dyn_collar_fr_l {parent=dyn_collar_fr_base_l, loc=<0 0 0>, rot=<0 0 0 -1>}
 dyn_collar_side_base_l {parent=neck_01, loc=<1.5362 -3.369 -15.247>, rot=<-0.0044 -1 0 0>}
 dyn_collar_side_l {parent=dyn_collar_side_base_l, loc=<0 0 0>, rot=<0 0 0 -1>}
 dyn_collar_bk_base_l {parent=neck_01, loc=<2.6262 -10.887 -10.3697>, rot=<-0.0796 -0.8722 0.4716 0.1025>}
 dyn_collar_bk_l {parent=dyn_collar_bk_base_l, loc=<0 0 0>, rot=<0 0 0 -1>}
 dyn_collar_bk_base {parent=neck_01, loc=<2.6193 -13.2386 0>, rot=<-0.1038 -0.6994 0.6994 0.1038>}
 dyn_collar_bk {parent=dyn_collar_bk_base, loc=<0 0 0>, rot=<0 0 0 -1>}
 dyn_collar_fr_base_r {parent=neck_01, loc=<-2.3711 12.1764 9.7943>, rot=<0.9588 -0.0042 -0.0012 0.284>}
 dyn_collar_fr_r {parent=dyn_collar_fr_base_r, loc=<0 0 0>, rot=<0 0 0 -1>}
 dyn_collar_side_base_r {parent=neck_01, loc=<2.5 0.7132 15.295>, rot=<1 -0.0044 0 0>}
 dyn_collar_side_r {parent=dyn_collar_side_base_r, loc=<0 0 0>, rot=<0 0 0 -1>}
 dyn_collar_bk_base_r {parent=neck_01, loc=<2.0818 -10.0365 11.5441>, rot=<0.8722 -0.0796 0.1025 -0.4716>}
 dyn_collar_bk_r {parent=dyn_collar_bk_base_r, loc=<0 0 0>, rot=<0 0 0 1>}
 dyn_collar_fr_flap_base_l {parent=spine_05, loc=<2.911 14.5936 -6.1764>, rot=<0.0332 -0.8272 -0.5604 0.0225>}
 dyn_collar_fr_flap_l {parent=dyn_collar_fr_flap_base_l, loc=<0 0 0>, rot=<0 0 0 1>}
 dyn_grenade_a_base {parent=spine_04, loc=<-1.0035 18.3681 -0.1673>, rot=<0.0416 -0.999 -0.0152 0.0006>}
 dyn_grenade_a {parent=dyn_grenade_a_base, loc=<0 0 0>, rot=<0 0 0 -1>}
 dyn_grenade_b_base {parent=spine_04, loc=<-5.6747 17.2248 -7.2359>, rot=<0.0414 -0.9936 0.1053 -0.0044>}
 dyn_grenade_b {parent=dyn_grenade_b_base, loc=<0 0 0>, rot=<0 0 0 -1>}
 dyn_grenade_c_base {parent=spine_04, loc=<-9.4732 12.3037 -13.6875>, rot=<0.0391 -0.9394 0.3402 -0.0142>}
 dyn_grenade_c {parent=dyn_grenade_c_base, loc=<0 0 0>, rot=<0 0 0 -1>}
 dyn_shoulder_flap_a_base_l {parent=spine_04, loc=<-2.2057 6.4422 -18.5892>, rot=<-0.0358 -0.793 0.5646 0.2261>}
 dyn_shoulder_flap_a_l {parent=dyn_shoulder_flap_a_base_l, loc=<0 0 0>, rot=<0 0 0 1>}
 dyn_shoulder_flap_b_base_l {parent=spine_04, loc=<-2.1863 -9.4453 -19.7501>, rot=<0.2708 -0.5464 0.7639 0.2109>}
 dyn_shoulder_flap_b_l {parent=dyn_shoulder_flap_b_base_l, loc=<0 0 0>, rot=<0 0 0 -1>}
 dyn_shoulder_flap_c_base_l {parent=spine_04, loc=<8.3855 -15.1623 -18.3473>, rot=<0.0415 -0.9973 0.0025 0.0605>}
 dyn_shoulder_flap_c_l {parent=dyn_shoulder_flap_c_base_l, loc=<0 0 0>, rot=<0 0 0 1>}
 dyn_shoulder_flap_b_base_r {parent=spine_04, loc=<10.2383 -15.399 19.742>, rot=<0.0474 -0.9928 0.0757 -0.0792>}
 dyn_shoulder_flap_b_r {parent=dyn_shoulder_flap_b_base_r, loc=<0 0 0>, rot=<0 0 0 -1>}
 thigh_l {parent=pelvis, loc=<-1.8716 -0.0109 -11.9936>, rot=<0.0012 0.0572 0.0155 -0.9982>}
 calf_l {parent=thigh_l, loc=<-41.9178 -0.2246 0.0926>, rot=<0.0006 -0.0161 -0.0391 -0.9991>}
 calf_twist_02_l {parent=calf_l, loc=<-15.9268 -0.0967 0.0397>, rot=<0.0237 0.0095 -0.0035 -0.9997>}
 foot_l {parent=calf_l, loc=<-38.7057 1.2438 0.2453>, rot=<0.0975 -0.0474 0.0317 -0.9936>}
 ball_l {parent=foot_l, loc=<-6.1614 13.1469 0.6807>, rot=<0.0001 -0.0001 -0.7063 -0.7079>}
 thigh_twist_01_l {parent=thigh_l, loc=<-21.6303 0 0>, rot=<0.0013 0 0 -1>}
 thigh_r {parent=pelvis, loc=<-1.8714 -0.0109 11.9936>, rot=<0.0572 -0.0012 0.9982 0.0155>}
 calf_r {parent=thigh_r, loc=<41.918 0.2246 -0.0926>, rot=<0.0006 -0.0161 -0.0391 -0.9991>}
 calf_twist_02_r {parent=calf_r, loc=<15.9268 0.0967 -0.0397>, rot=<0.0237 0.0095 -0.0035 -0.9997>}
 foot_r {parent=calf_r, loc=<38.7059 -1.2438 -0.2453>, rot=<0.0975 -0.0474 0.0317 -0.9936>}
 ball_r {parent=foot_r, loc=<6.1614 -13.1469 -0.6807>, rot=<0.0001 -0.0001 -0.7063 -0.7079>}
 dyn_ribbon_a_base_r {parent=foot_r, loc=<1.1041 -6.4639 -2.3086>, rot=<0.9139 0.399 0.0492 -0.0565>}
 dyn_ribbon_a_r {parent=dyn_ribbon_a_base_r, loc=<0 0 0>, rot=<0 0 0 -1>}
 dyn_ribbon_b_base_r {parent=foot_r, loc=<1.0107 -6.4786 1.2471>, rot=<0.901 0.4326 0.0027 -0.0316>}
 dyn_ribbon_b_r {parent=dyn_ribbon_b_base_r, loc=<0 0 0>, rot=<0 0 0 -1>}
 thigh_twist_01_r {parent=thigh_r, loc=<21.6303 0 0>, rot=<0.0013 0 0 -1>}
 dyn_strap_body_base {parent=pelvis, loc=<4.8648 4.9831 16.413>, rot=<-0.0392 -0.9561 0.28 0.077>}
 dyn_strap_body {parent=dyn_strap_body_base, loc=<0 0 0>, rot=<0 0 0 -1>}
 dyn_strapA_a_base {parent=dyn_strap_body, loc=<3.7154 1.5446 0.2614>, rot=<0.6412 0.0795 -0.0302 -0.7626>}
 dyn_strapA_a {parent=dyn_strapA_a_base, loc=<0 0 0>, rot=<0 0 0 -1>}
 dyn_strapA_b {parent=dyn_strapA_a, loc=<7.5207 0.5616 -0.2408>, rot=<-0.4774 0 0 -0.8787>}
 dyn_strapA_c {parent=dyn_strapA_b, loc=<6.4089 0.2659 -0.3724>, rot=<-0.0009 0.0001 0.0335 -0.9994>}
 dyn_strapB_a_base {parent=dyn_strap_body, loc=<3.2333 -2.256 0.3446>, rot=<0.845 -0.1016 0.1111 -0.5131>}
 dyn_strapB_a {parent=dyn_strapB_a_base, loc=<0 0 0>, rot=<0 0 0 -1>}
 dyn_strapB_b {parent=dyn_strapB_a, loc=<8.555 0.0429 0.1806>, rot=<0.3824 0.0043 0.0104 -0.9239>}
 dyn_strapB_c {parent=dyn_strapB_b, loc=<6.377 -0.2411 -0.083>, rot=<0 0 0.0579 -0.9983>}
]
