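# Physics asset for the large male Choco Bunny player body cosmetic, defining secondary motion for its collar, gloves, shoulder flaps, grenades and straps.
# Each dyn_ bone has a capsule body, with a constraint tying it to a small base-bone capsule at narrow, per-piece swing and twist limits.
# The two straps are three-segment chains with looser swing.
# Head and right thigh capsules have no constraints, apparently serving as colliders.
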PhysicsAsset /BRCosmetics/Characters/Player/Male/Large/Bodies/M_LRG_Choco_Bunny/Meshes/M_LRG_Choco_Bunny_Physics
bodies {
 dyn_collar_bk [{shapeType=Capsule, length=9, radius=2.25, center=<-4.1749 0 0>, rotation=<rot 87.2266 -0.0007 -0.0003>}]
 dyn_collar_bk_base [{shapeType=Capsule, length=1, radius=0.5}]
 dyn_collar_bk_base_l [{shapeType=Capsule, length=1, radius=0.5}]
 dyn_collar_bk_base_r [{shapeType=Capsule, length=1, radius=0.5}]
 dyn_collar_bk_l [{shapeType=Capsule, length=5, radius=2.5, center=<-3.2063 0 -0.0001>, rotation=<rot 90 0 0>}]
 dyn_collar_bk_r [{shapeType=Capsule, length=6.25, radius=3, center=<3.0448 0 0>, rotation=<rot 90 0 0>}]
 dyn_collar_fr_base_l [{shapeType=Capsule, length=1, radius=0.5}]
 dyn_collar_fr_base_r [{shapeType=Capsule, length=1, radius=0.5}]
 dyn_collar_fr_flap_base_l [{shapeType=Capsule, length=1, radius=0.5}]
 dyn_collar_fr_flap_l [{shapeType=Capsule, length=2.5, radius=1.5, center=<2.5658 0 0>, rotation=<rot 90 0 0>}]
 dyn_collar_fr_l [{shapeType=Capsule, length=3.75, radius=2, center=<-2.5501 0 0>, rotation=<rot 84.0929 0.0001 0.0001>}]
 dyn_collar_fr_r [{shapeType=Capsule, length=5.5, radius=2, center=<3.1126 0 -0.1215>, rotation=<rot 77.2292 0 0>}]
 dyn_collar_side_base_l [{shapeType=Capsule, length=1, radius=0.5}]
 dyn_collar_side_base_r [{shapeType=Capsule, length=1, radius=0.5}]
 dyn_collar_side_l [{shapeType=Capsule, length=5.5, radius=2, center=<-3.8086 0 0>, rotation=<rot 80.8199 0 0>}]
 dyn_collar_side_r [{shapeType=Capsule, length=5.5, radius=2, center=<2 0 0>, rotation=<rot 76.6226 0 0>}]
 dyn_glove_a_base_l [{shapeType=Capsule, length=0.505, radius=0.505, rotation=<rot 0 -90 0>}]
 dyn_glove_a_base_r [{shapeType=Capsule, length=0.505, radius=0.505, rotation=<rot 0 -90 180>}]
 dyn_glove_a_l [{shapeType=Capsule, length=1, radius=1, center=<1.1829 0 0>, rotation=<rot 90 0 0>}]
 dyn_glove_a_r [{shapeType=Capsule, length=1, radius=1, center=<-1.1829 0 0>, rotation=<rot 90 0 180>}]
 dyn_glove_b_base_l [{shapeType=Capsule, length=0.505, radius=0.505, rotation=<rot 0 -90 0>}]
 dyn_glove_b_base_r [{shapeType=Capsule, length=0.505, radius=0.505, rotation=<rot 0 -90 180>}]
 dyn_glove_b_l [{shapeType=Capsule, length=1, radius=1, center=<1 0 0>, rotation=<rot 90 0 0>}]
 dyn_glove_b_r [{shapeType=Capsule, length=1, radius=1, center=<-1 0 0>, rotation=<rot 90 0 180>}]
 dyn_glove_c_base_l [{shapeType=Capsule, length=0.505, radius=0.505, rotation=<rot 0 -90 0>}]
 dyn_glove_c_base_r [{shapeType=Capsule, length=0.505, radius=0.505, rotation=<rot 0 -90 180>}]
 dyn_glove_c_l [{shapeType=Capsule, length=1, radius=1, center=<1 0 0>, rotation=<rot 90 0 0>}]
 dyn_glove_c_r [{shapeType=Capsule, length=1, radius=1, center=<-1 0 0>, rotation=<rot 90 0 180>}]
 dyn_glove_d_base_l [{shapeType=Capsule, length=0.505, radius=0.505, rotation=<rot 0 -90 0>}]
 dyn_glove_d_base_r [{shapeType=Capsule, length=0.505, radius=0.505, rotation=<rot 0 -90 180>}]
 dyn_glove_d_l [{shapeType=Capsule, length=1, radius=1, center=<1 0 0>, rotation=<rot 90 0 0>}]
 dyn_glove_d_r [{shapeType=Capsule, length=1, radius=1, center=<-1 0 0>, rotation=<rot 90 0 180>}]
 dyn_glove_e_base_l [{shapeType=Capsule, length=0.505, radius=0.505, rotation=<rot 0 -90 0>}]
 dyn_glove_e_base_r [{shapeType=Capsule, length=0.505, radius=0.505, rotation=<rot 0 -90 180>}]
 dyn_glove_e_l [{shapeType=Capsule, length=1, radius=1, center=<1 0 0>, rotation=<rot 90 0 0>}]
 dyn_glove_e_r [{shapeType=Capsule, length=1, radius=1, center=<-1 0 0>, rotation=<rot 90 0 180>}]
 dyn_grenade_a [{shapeType=Capsule, length=8.5541, radius=5.7703, center=<-0.3178 3.7932 0.4753>, rotation=<rot 0 -75.0015 -68.4458>}]
 dyn_grenade_a_base [{shapeType=Capsule, length=1, radius=1}]
 dyn_grenade_b [{shapeType=Capsule, length=4.5, radius=4.75, center=<2.4327 0.0001 0>, rotation=<rot 90 0 0>}]
 dyn_grenade_b_base [{shapeType=Capsule, length=1, radius=1}]
 dyn_grenade_c [{shapeType=Capsule, length=8.5538, radius=6.2677, center=<-0.2606 4.7199 0.1767>, rotation=<rot 0 -88.357 -74.6406>}]
 dyn_grenade_c_base [{shapeType=Capsule, length=1, radius=1}]
 dyn_shoulder_flap_a_base_l [{shapeType=Capsule, length=1, radius=1}]
 dyn_shoulder_flap_a_l [{shapeType=Capsule, length=1.75, radius=2, center=<2 0 0>, rotation=<rot 90 0 0>}]
 dyn_shoulder_flap_b_base_l [{shapeType=Capsule, length=1, radius=1}]
 dyn_shoulder_flap_b_base_r [{shapeType=Capsule, length=1, radius=1}]
 dyn_shoulder_flap_b_l [{shapeType=Capsule, length=1.75, radius=2, center=<2 0 0>, rotation=<rot 90 0 0>}]
 dyn_shoulder_flap_b_r [{shapeType=Capsule, length=3.75, radius=2, center=<3.4613 0 0>, rotation=<rot 90 0 0>}]
 dyn_shoulder_flap_c_base_l [{shapeType=Capsule, length=1, radius=1}]
 dyn_shoulder_flap_c_l [{shapeType=Capsule, length=1.75, radius=2, center=<2 0 0>, rotation=<rot 90 0 0>}]
 dyn_strapA_a [{shapeType=Capsule, length=3.8428, radius=1.6252, center=<3.7639 0.0889 0>, rotation=<rot 90 -7.125 -12.9907>}]
 dyn_strapA_a_base [{shapeType=Capsule, length=1, radius=1}]
 dyn_strapA_b [{shapeType=Capsule, length=3.4124, radius=2.2007, center=<3.2818 0 0>, rotation=<rot 90 0 0>}]
 dyn_strapA_c [{shapeType=Capsule, length=4.6828, radius=2.376, center=<4.5686 -0.0001 -0.1292>, rotation=<rot 88.3801 -0.0008 -0.0008>}]
 dyn_strapB_a [{shapeType=Capsule, length=5.75, radius=1, center=<4.3588 0.1123 0>, rotation=<rot 90 -0.8952 1.7903>}]
 dyn_strapB_a_base [{shapeType=Capsule, length=1, radius=1}]
 dyn_strapB_b [{shapeType=Capsule, length=3.75, radius=1.5, center=<3.1672 0 0>, rotation=<rot 90 34.5085 39.0954>}]
 dyn_strapB_c [{shapeType=Capsule, length=4.75, radius=1.75, center=<4.051 0 0>, rotation=<rot 90 0 0>}]
 head [{shapeType=Capsule, length=0.5, radius=14, center=<-1.6708 0.1074 0>, rotation=<rot 90 0 0>}]
 thigh_r [{shapeType=Capsule, length=28.25, radius=8.75, center=<13.5658 0.0317 -2.9311>, rotation=<rot 90 0 3.3613>}]
}
constraints [
 {bone1Name=dyn_strapA_a, bone2Name=dyn_strapA_a_base, swing1LimitDegrees=45, swing2LimitDegrees=45, twistLimitDegrees=10}
 {bone1Name=dyn_strapA_b, bone2Name=dyn_strapA_a, swing1LimitDegrees=45, swing2LimitDegrees=45, twistLimitDegrees=10}
 {bone1Name=dyn_strapA_c, bone2Name=dyn_strapA_b, swing1LimitDegrees=45, swing2LimitDegrees=45, twistLimitDegrees=10}
 {bone1Name=dyn_strapB_a, bone2Name=dyn_strapB_a_base, swing1LimitDegrees=50, swing2LimitDegrees=50, twistLimitDegrees=10}
 {bone1Name=dyn_strapB_b, bone2Name=dyn_strapB_a, swing1LimitDegrees=50, swing2LimitDegrees=50, twistLimitDegrees=10}
 {bone1Name=dyn_strapB_c, bone2Name=dyn_strapB_b, swing1LimitDegrees=50, swing2LimitDegrees=50, twistLimitDegrees=10}
 {bone1Name=dyn_grenade_a, bone2Name=dyn_grenade_a_base, swing1LimitDegrees=10, swing2LimitDegrees=10, twistLimitDegrees=10}
 {bone1Name=dyn_grenade_b, bone2Name=dyn_grenade_b_base, swing1LimitDegrees=15, swing2LimitDegrees=15, twistLimitDegrees=15}
 {bone1Name=dyn_grenade_c, bone2Name=dyn_grenade_c_base, swing1LimitDegrees=10, swing2LimitDegrees=10, twistLimitDegrees=10}
 {bone1Name=dyn_collar_fr_l, bone2Name=dyn_collar_fr_base_l, swing1LimitDegrees=5, swing2LimitDegrees=5, twistLimitDegrees=5}
 {bone1Name=dyn_collar_side_l, bone2Name=dyn_collar_side_base_l, swing1LimitDegrees=5, swing2LimitDegrees=5, twistLimitDegrees=6}
 {bone1Name=dyn_collar_bk_l, bone2Name=dyn_collar_bk_base_l, swing1LimitDegrees=5, swing2LimitDegrees=5, twistLimitDegrees=0.2}
 {bone1Name=dyn_collar_bk, bone2Name=dyn_collar_bk_base, swing1LimitDegrees=2, swing2LimitDegrees=2, twistLimitDegrees=0.2}
 {bone1Name=dyn_collar_fr_r, bone2Name=dyn_collar_fr_base_r, swing1LimitDegrees=5, swing2LimitDegrees=5, twistLimitDegrees=5}
 {bone1Name=dyn_collar_side_r, bone2Name=dyn_collar_side_base_r, swing1LimitDegrees=5, swing2LimitDegrees=5, twistLimitDegrees=5}
 {bone1Name=dyn_collar_bk_r, bone2Name=dyn_collar_bk_base_r, swing1LimitDegrees=5, swing2LimitDegrees=5, twistLimitDegrees=0.2}
 {bone1Name=dyn_collar_fr_flap_l, bone2Name=dyn_collar_fr_flap_base_l, swing1LimitDegrees=5, swing2LimitDegrees=25, twistLimitDegrees=5}
 {bone1Name=dyn_glove_b_l, bone2Name=dyn_glove_b_base_l, swing1LimitDegrees=5, swing2LimitDegrees=5, twistLimitDegrees=5}
 {bone1Name=dyn_glove_c_l, bone2Name=dyn_glove_c_base_l, swing1LimitDegrees=5, swing2LimitDegrees=5, twistLimitDegrees=5}
 {bone1Name=dyn_glove_d_l, bone2Name=dyn_glove_d_base_l, swing1LimitDegrees=25, swing2LimitDegrees=10, twistLimitDegrees=5}
 {bone1Name=dyn_glove_e_l, bone2Name=dyn_glove_e_base_l, swing1LimitDegrees=5, swing2LimitDegrees=5, twistLimitDegrees=5}
 {bone1Name=dyn_glove_a_l, bone2Name=dyn_glove_a_base_l, swing1LimitDegrees=25, swing2LimitDegrees=10, twistLimitDegrees=5}
 {bone1Name=dyn_shoulder_flap_a_l, bone2Name=dyn_shoulder_flap_a_base_l, swing1LimitDegrees=15, swing2LimitDegrees=5, twistLimitDegrees=5}
 {bone1Name=dyn_shoulder_flap_b_l, bone2Name=dyn_shoulder_flap_b_base_l, swing1LimitDegrees=15, swing2LimitDegrees=5, twistLimitDegrees=5}
 {bone1Name=dyn_shoulder_flap_c_l, bone2Name=dyn_shoulder_flap_c_base_l, swing1LimitDegrees=3, swing2LimitDegrees=2, twistLimitDegrees=0.2}
 {bone1Name=dyn_shoulder_flap_b_r, bone2Name=dyn_shoulder_flap_b_base_r, swing1LimitDegrees=5, swing2LimitDegrees=5, twistLimitDegrees=2}
 {bone1Name=dyn_glove_a_r, bone2Name=dyn_glove_a_base_r, swing1LimitDegrees=25, swing2LimitDegrees=10, twistLimitDegrees=5}
 {bone1Name=dyn_glove_b_r, bone2Name=dyn_glove_b_base_r, swing1LimitDegrees=5, swing2LimitDegrees=5, twistLimitDegrees=5}
 {bone1Name=dyn_glove_c_r, bone2Name=dyn_glove_c_base_r, swing1LimitDegrees=5, swing2LimitDegrees=5, twistLimitDegrees=5}
 {bone1Name=dyn_glove_d_r, bone2Name=dyn_glove_d_base_r, swing1LimitDegrees=25, swing2LimitDegrees=10, twistLimitDegrees=5}
 {bone1Name=dyn_glove_e_r, bone2Name=dyn_glove_e_base_r, swing1LimitDegrees=5, swing2LimitDegrees=5, twistLimitDegrees=5}
]
