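# Physics asset for the ribbon of the large male Choco Bunny player body, covering two right-side ribbon strands, A and B.
# Each strand has a tiny capsule on its base bone and a larger, offset capsule on the ribbon bone.
# One constraint per strand ties ribbon to base with identical narrow swing and twist limits, with no constraint linking the two strands.
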PhysicsAsset /BRCosmetics/Characters/Player/Male/Large/Bodies/M_LRG_Choco_Bunny/Meshes/M_LRG_Choco_Bunny_Ribbon_Physics
bodies {
 dyn_ribbon_a_base_r [{shapeType=Capsule, length=0.5, radius=0.5}]
 dyn_ribbon_a_r [{shapeType=Capsule, length=1.5, radius=2, center=<-0.0001 0 1.9222>}]
 dyn_ribbon_b_base_r [{shapeType=Capsule, length=0.5, radius=0.5}]
 dyn_ribbon_b_r [{shapeType=Capsule, length=1.5, radius=2, center=<0 0 -1.521>}]
}
constraints [
 {bone1Name=dyn_ribbon_a_r, bone2Name=dyn_ribbon_a_base_r, swing1LimitDegrees=20, swing2LimitDegrees=30, twistLimitDegrees=15}
 {bone1Name=dyn_ribbon_b_r, bone2Name=dyn_ribbon_b_base_r, swing1LimitDegrees=20, swing2LimitDegrees=30, twistLimitDegrees=15}
]
