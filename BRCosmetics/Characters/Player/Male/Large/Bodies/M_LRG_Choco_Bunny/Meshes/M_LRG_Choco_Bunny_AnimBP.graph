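# Post-processes the Choco Bunny large male body cosmetic's pose, adding secondary motion to grenades, straps, collar and flaps.
# Blends the attached parent's pose with the pawn's head mesh at face and jaw branches.
# Weighted constraints and bone offsets, some driven by right-thigh angles read each update, feed two RigidBody simulations, then a Control Rig.
# The glove constraint chain and a third RigidBody are unconnected; left-leg and head angles go unused.
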AnimBlueprint /BRCosmetics/Characters/Player/Male/Large/Bodies/M_LRG_Choco_Bunny/Meshes/M_LRG_Choco_Bunny_AnimBP
target_skeleton=&/M_LRG_Choco_Bunny_Skeleton
graphs [
 AnimGraph {nodes [
   n1 = "Constraint - dyn_glove_a_base_l (TTTTT)"!(
    ComponentPose
    Alpha=1
    ConstraintWeights_0=0.3892
    ConstraintWeights_1:float=0.4393
    ConstraintWeights_2:float=0.0332
    ConstraintWeights_3:float=0.0401
    ConstraintWeights_4:float=0.0982
   ) -> (Pose) {
    alphaBoolBlend {blendOption=Linear}
    boneToModify {boneName=dyn_glove_a_base_l}
    constraintSetup [
     {transformType=Translation, targetBone {boneName=hand_l}}
     {transformType=Translation, targetBone {boneName=lowerarm_twist_01_l}}
     {transformType=Translation, targetBone {boneName=index_metacarpal_l}}
     {transformType=Translation, targetBone {boneName=ring_metacarpal_l}}
     {transformType=Translation, targetBone {boneName=thumb_01_l}}
    ]
    constraintWeights [1,1,1,1,1]
   }
   n2 = "Constraint - dyn_glove_b_base_l (TTTT)"!(
    ComponentPose=@n1
    Alpha=1
    ConstraintWeights_0=0.3997
    ConstraintWeights_1:float=0.5508
    ConstraintWeights_2:float=0.0318
    ConstraintWeights_3:float=0.0177
   ) -> (Pose) {
    alphaBoolBlend {blendOption=Linear}
    boneToModify {boneName=dyn_glove_b_base_l}
    constraintSetup [
     {transformType=Translation, targetBone {boneName=hand_l}}
     {transformType=Translation, targetBone {boneName=lowerarm_twist_01_l}}
     {transformType=Translation, targetBone {boneName=pinky_metacarpal_l}}
     {transformType=Translation, targetBone {boneName=ring_metacarpal_l}}
    ]
    constraintWeights [1,1,1,1]
   }
   n3 = "Constraint - dyn_glove_c_base_l (TTT)"!(ComponentPose=@n2, Alpha=1, ConstraintWeights_0=0.2956, ConstraintWeights_1:float=0.603, ConstraintWeights_2:float=0.1014) -> (Pose) {
    alphaBoolBlend {blendOption=Linear}
    boneToModify {boneName=dyn_glove_c_base_l}
    constraintSetup [
     {transformType=Translation, targetBone {boneName=hand_l}}
     {transformType=Translation, targetBone {boneName=lowerarm_twist_01_l}}
     {transformType=Translation, targetBone {boneName=pinky_metacarpal_l}}
    ]
    constraintWeights [1,1,1]
   }
   n4 = "Constraint - dyn_glove_d_base_l (TTTTT)"!(
    ComponentPose=@n3
    Alpha=1
    ConstraintWeights_0=0.2026
    ConstraintWeights_1:float=0.5416
    ConstraintWeights_2:float=0.0381
    ConstraintWeights_3:float=0.0428
    ConstraintWeights_4:float=0.175
   ) -> (Pose) {
    alphaBoolBlend {blendOption=Linear}
    boneToModify {boneName=dyn_glove_d_base_l}
    constraintSetup [
     {transformType=Translation, targetBone {boneName=hand_l}}
     {transformType=Translation, targetBone {boneName=lowerarm_twist_01_l}}
     {transformType=Translation, targetBone {boneName=pinky_metacarpal_l}}
     {transformType=Translation, targetBone {boneName=lowerarm_twist_02_l}}
     {transformType=Translation, targetBone {boneName=thumb_01_l}}
    ]
    constraintWeights [1,1,1,1,1]
   }
   "Constraint - dyn_glove_e_base_l (TTTT)"!(
    ComponentPose=@n4
    Alpha=1
    ConstraintWeights_0=0.1226
    ConstraintWeights_1:float=0.4259
    ConstraintWeights_2:float=0.0659
    ConstraintWeights_3:float=0.3856
   ) -> (Pose) {
    alphaBoolBlend {blendOption=Linear}
    boneToModify {boneName=dyn_glove_e_base_l}
    constraintSetup [
     {transformType=Translation, targetBone {boneName=hand_l}}
     {transformType=Translation, targetBone {boneName=lowerarm_twist_01_l}}
     {transformType=Translation, targetBone {boneName=lowerarm_twist_02_l}}
     {transformType=Translation, targetBone {boneName=thumb_01_l}}
    ]
    constraintWeights [1,1,1,1]
   }
   RigidBody() {
    bClampLinearTranslationLimitToRefPose=true
    bOverrideWorldGravity=true
    overridePhysicsAsset=&/M_LRG_Choco_Bunny_Physics
    simulationSpace=BaseBoneSpace
    alphaBoolBlend {blendOption=Linear}
    baseBoneRef {boneName=spine_01}
    componentLinearAccScale=<0.2 0.2 0.3>
    componentLinearVelScale=<0.2 0.2 0.3>
    overrideWorldGravity {z=-980}
   }
   n5 = Copy Pose From Mesh() {bUseAttachedParent=true}
   n6 = Copy Pose From Mesh(SourceMeshComponent=Get MeshToCopy() -> (MeshToCopy:Skeletal Mesh Component Object Reference)) {bCopyCurves=true}
   n7 = "Constraint - dyn_grenade_a_base (PPPPPPPP)"!(
    ComponentPose=Local To Component(LocalPose=Layered blend per bone(BasePose=@n5, BlendPoses_0=@n6) {layerSetup [{branchFilters [{boneName=faceAttach}, {boneName=C_jaw}]}]})
    Alpha=1
    ConstraintWeights_0=0.0371
    ConstraintWeights_1:float=0.046
    ConstraintWeights_2:float=0.179
    ConstraintWeights_3:float=0.2204
    ConstraintWeights_4:float=0.2396
    ConstraintWeights_5:float=0.1367
    ConstraintWeights_6:float=0.0464
    ConstraintWeights_7:float=0.0947
   ) -> (Pose) {
    alphaBoolBlend {blendOption=Linear}
    boneToModify {boneName=dyn_grenade_a_base}
    constraintSetup [
     {targetBone {boneName=clavicle_l}}
     {targetBone {boneName=spine_01}}
     {targetBone {boneName=spine_02}}
     {targetBone {boneName=spine_03}}
     {targetBone {boneName=spine_04}}
     {targetBone {boneName=spine_05}}
     {targetBone {boneName=clavicle_r}}
     {targetBone {boneName=pelvis}}
    ]
    constraintWeights [1,1,1,1,1,1,1,1]
   }
   n8 = "Constraint - dyn_grenade_b_base (PPPPPPP)"!(
    ComponentPose=@n7
    Alpha=1
    ConstraintWeights_0=0.0785
    ConstraintWeights_1:float=0.0873
    ConstraintWeights_2:float=0.2432
    ConstraintWeights_3:float=0.2405
    ConstraintWeights_4:float=0.154
    ConstraintWeights_5:float=0.0302
    ConstraintWeights_6:float=0.1662
   ) -> (Pose) {
    alphaBoolBlend {blendOption=Linear}
    boneToModify {boneName=dyn_grenade_b_base}
    constraintSetup [
     {targetBone {boneName=clavicle_l}}
     {targetBone {boneName=spine_01}}
     {targetBone {boneName=spine_02}}
     {targetBone {boneName=spine_03}}
     {targetBone {boneName=spine_04}}
     {targetBone {boneName=spine_05}}
     {targetBone {boneName=pelvis}}
    ]
    constraintWeights [1,1,1,1,1,1,1]
   }
   n9 = "Constraint - dyn_grenade_c_base (PPPPPP)"!(
    ComponentPose=@n8
    Alpha=1
    ConstraintWeights_0=0.0698
    ConstraintWeights_1:float=0.1259
    ConstraintWeights_2:float=0.2542
    ConstraintWeights_3:float=0.2343
    ConstraintWeights_4:float=0.0813
    ConstraintWeights_5:float=0.2345
   ) -> (Pose) {
    alphaBoolBlend {blendOption=Linear}
    boneToModify {boneName=dyn_grenade_c_base}
    constraintSetup [
     {targetBone {boneName=clavicle_l}}
     {targetBone {boneName=spine_01}}
     {targetBone {boneName=spine_02}}
     {targetBone {boneName=spine_03}}
     {targetBone {boneName=spine_04}}
     {targetBone {boneName=pelvis}}
    ]
    constraintWeights [1,1,1,1,1,1]
   }
   n10 = "Constraint - dyn_strap_body_base (PPPPP)"!(
    ComponentPose=@n9
    Alpha=1
    ConstraintWeights_0=0.1259
    ConstraintWeights_1:float=0.0884
    ConstraintWeights_2:float=0.0601
    ConstraintWeights_3:float=0.015
    ConstraintWeights_4:float=0.7106
   ) -> (Pose) {
    alphaBoolBlend {blendOption=Linear}
    boneToModify {boneName=dyn_strap_body_base}
    constraintSetup [
     {targetBone {boneName=spine_01}}
     {targetBone {boneName=spine_02}}
     {targetBone {boneName=spine_03}}
     {targetBone {boneName=thigh_twist_01_r}}
     {targetBone {boneName=pelvis}}
    ]
    constraintWeights [1,1,1,1,1]
   }
   n11 = "Transform (Modify) Bone - Bone: dyn_strap_body_base"(ComponentPose=@n10, Alpha=Get legUp_R() -> (legUp_R:double), Translation="-2,0,0", Rotation="-20,40,0", Scale="1,1,1") {
    rotationMode=BMM_Additive
    rotationSpace=BCS_BoneSpace
    translationMode=BMM_Additive
    translationSpace=BCS_BoneSpace
    alphaBoolBlend {blendOption=Linear}
    alphaScaleBiasClamp {bMapRange=true, inRange {max=0, min=-100}, outRange {max=0, min=1}}
    boneToModify {boneName=dyn_strap_body_base}
    rotation {pitch=-20, yaw=40}
    translation {x=-2}
   }
   n12 = "Constraint - dyn_strapA_a_base (R)"(ComponentPose=@n11, Alpha=1, ConstraintWeights_0=1) {
    alphaBoolBlend {blendOption=Linear}
    boneToModify {boneName=dyn_strapA_a_base}
    constraintSetup [{transformType=Rotation, targetBone {boneName=thigh_r}}]
    constraintWeights [1]
   }
   n13 = "Transform (Modify) Bone - Bone: dyn_strapA_a_base"(ComponentPose=@n12, Alpha=Get legUp_R(), Translation="0,3,0", Rotation="0,0,0", Scale="1,1,1") {
    rotationMode=BMM_Additive
    rotationSpace=BCS_BoneSpace
    translationMode=BMM_Additive
    translationSpace=BCS_BoneSpace
    alphaBoolBlend {blendOption=Linear}
    alphaScaleBiasClamp {bMapRange=true, inRange {max=0, min=-100}, outRange {max=0, min=1}}
    boneToModify {boneName=dyn_strapA_a_base}
    translation {y=3}
   }
   n14 = "Transform (Modify) Bone - Bone: dyn_strapB_a_base"(ComponentPose=@n13, Alpha=Get legUp_R(), Translation="0,2,0", Rotation="0,0,0", Scale="1,1,1") {
    rotationMode=BMM_Additive
    rotationSpace=BCS_BoneSpace
    translationMode=BMM_Additive
    translationSpace=BCS_BoneSpace
    alphaBoolBlend {blendOption=Linear}
    alphaScaleBiasClamp {bMapRange=true, inRange {max=0, min=-100}, outRange {max=0, min=1}}
    boneToModify {boneName=dyn_strapB_a_base}
    translation {y=2}
   }
   n15 = "Transform (Modify) Bone - Bone: dyn_strapB_a_base"(ComponentPose=@n14, Alpha=Get legSide_R() -> (legSide_R:double), Translation="0,0,0", Rotation="0,60,0", Scale="1,1,1") {
    rotationMode=BMM_Additive
    rotationSpace=BCS_BoneSpace
    translationMode=BMM_Additive
    translationSpace=BCS_BoneSpace
    alphaBoolBlend {blendOption=Linear}
    alphaScaleBiasClamp {bMapRange=true, inRange {max=40}}
    boneToModify {boneName=dyn_strapB_a_base}
    rotation {yaw=60}
   }
   n16 = "Transform (Modify) Bone - Bone: dyn_collar_bk_base"(ComponentPose=@n15, Alpha=1, Translation="0,0,4", Rotation="-8,0.000,0.000", Scale="1,1,1") {
    rotationMode=BMM_Additive
    rotationSpace=BCS_BoneSpace
    translationMode=BMM_Additive
    translationSpace=BCS_BoneSpace
    alphaBoolBlend {blendOption=Linear}
    alphaScaleBiasClamp {inRange {max=15}}
    boneToModify {boneName=dyn_collar_bk_base}
    rotation {pitch=-8}
    translation {z=4}
   }
   n17 = "Transform (Modify) Bone - Bone: dyn_collar_bk_base_l"(ComponentPose=@n16, Alpha=1, Translation="0,0,3.6", Rotation="-5,0,0", Scale="1,1,1") {
    rotationMode=BMM_Additive
    rotationSpace=BCS_BoneSpace
    translationMode=BMM_Additive
    translationSpace=BCS_BoneSpace
    alphaBoolBlend {blendOption=Linear}
    alphaScaleBiasClamp {inRange {max=15}}
    boneToModify {boneName=dyn_collar_bk_base_l}
    rotation {pitch=-5}
    translation {z=3.6}
   }
   n18 = "Transform (Modify) Bone - Bone: dyn_collar_bk_base_r"(ComponentPose=@n17, Alpha=1, Translation="0,0,-3.5", Rotation="-10,0,0", Scale="1,1,1") {
    rotationMode=BMM_Additive
    rotationSpace=BCS_BoneSpace
    translationMode=BMM_Additive
    translationSpace=BCS_BoneSpace
    alphaBoolBlend {blendOption=Linear}
    alphaScaleBiasClamp {inRange {max=15}}
    boneToModify {boneName=dyn_collar_bk_base_r}
    rotation {pitch=-10}
    translation {z=-3.5}
   }
   n19 = "Constraint - dyn_shoulder_flap_a_base_l (PPPPPPPP)"!(
    ComponentPose=@n18
    Alpha=1
    ConstraintWeights_0=0.2933
    ConstraintWeights_1:float=0.0632
    ConstraintWeights_2:float=0.0498
    ConstraintWeights_3:float=0.1198
    ConstraintWeights_4:float=0.2538
    ConstraintWeights_5:float=0.0927
    ConstraintWeights_6:float=0.0851
    ConstraintWeights_7:float=0.0424
   ) -> (Pose) {
    alphaBoolBlend {blendOption=Linear}
    boneToModify {boneName=dyn_shoulder_flap_a_base_l}
    constraintSetup [
     {targetBone {boneName=clavicle_l}}
     {targetBone {boneName=upperarm_twist_02_l}}
     {targetBone {boneName=spine_01}}
     {targetBone {boneName=spine_02}}
     {targetBone {boneName=spine_03}}
     {targetBone {boneName=spine_04}}
     {targetBone {boneName=upperarm_twist_01_l}}
     {targetBone {boneName=pelvis}}
    ]
    constraintWeights [1,1,1,1,1,1,1,1]
   }
   n20 = "Constraint - dyn_shoulder_flap_b_base_l (PPPPPPPP)"!(
    ComponentPose=@n19
    Alpha=1
    ConstraintWeights_0=0.2723
    ConstraintWeights_1:float=0.1466
    ConstraintWeights_2:float=0.0399
    ConstraintWeights_3:float=0.0824
    ConstraintWeights_4:float=0.2429
    ConstraintWeights_5:float=0.1122
    ConstraintWeights_6:float=0.0446
    ConstraintWeights_7:float=0.0591
   ) -> (Pose) {
    alphaBoolBlend {blendOption=Linear}
    boneToModify {boneName=dyn_shoulder_flap_b_base_l}
    constraintSetup [
     {targetBone {boneName=clavicle_l}}
     {targetBone {boneName=upperarm_twist_02_l}}
     {targetBone {boneName=spine_01}}
     {targetBone {boneName=spine_02}}
     {targetBone {boneName=spine_03}}
     {targetBone {boneName=upperarm_twist_01_l}}
     {targetBone {boneName=pelvis}}
     {targetBone {boneName=upperarm_l}}
    ]
    constraintWeights [1,1,1,1,1,1,1,1]
   }
   n21 = "Constraint - dyn_shoulder_flap_c_base_l (PPPPPPPP)"!(
    ComponentPose=@n20
    Alpha=1
    ConstraintWeights_0=0.4703
    ConstraintWeights_1:float=0.0763
    ConstraintWeights_2:float=0.034
    ConstraintWeights_3:float=0.2105
    ConstraintWeights_4:float=0.0367
    ConstraintWeights_5:float=0.0206
    ConstraintWeights_6:float=0.1212
    ConstraintWeights_7:float=0.0304
   ) -> (Pose) {
    alphaBoolBlend {blendOption=Linear}
    boneToModify {boneName=dyn_shoulder_flap_c_base_l}
    constraintSetup [
     {targetBone {boneName=clavicle_l}}
     {targetBone {boneName=upperarm_twist_02_l}}
     {targetBone {boneName=spine_02}}
     {targetBone {boneName=spine_03}}
     {targetBone {boneName=spine_04}}
     {targetBone {boneName=spine_05}}
     {targetBone {boneName=upperarm_twist_01_l}}
     {targetBone {boneName=upperarm_l}}
    ]
    constraintWeights [1,1,1,1,1,1,1,1]
   }
   n22 = "Transform (Modify) Bone - Bone: dyn_shoulder_flap_c_base_l"(ComponentPose=@n21, Alpha=1, Translation="0.000,-3.4,0", Rotation="0.000,0.000,2.4", Scale="1,1,1") {
    rotationMode=BMM_Additive
    rotationSpace=BCS_BoneSpace
    translationMode=BMM_Additive
    translationSpace=BCS_BoneSpace
    alphaBoolBlend {blendOption=Linear}
    alphaScaleBiasClamp {inRange {max=15}}
    boneToModify {boneName=dyn_shoulder_flap_c_base_l}
    rotation {roll=2.4}
    translation {y=-3.4}
   }
   n23 = "Constraint - dyn_shoulder_flap_b_base_r (PPPPPPPP)"!(
    ComponentPose=@n22
    Alpha=1
    ConstraintWeights_0=0.0501
    ConstraintWeights_1:float=0.1667
    ConstraintWeights_2:float=0.109
    ConstraintWeights_3:float=0.0193
    ConstraintWeights_4:float=0.1521
    ConstraintWeights_5:float=0.021
    ConstraintWeights_6:float=0.0178
    ConstraintWeights_7:float=0.464
   ) -> (Pose) {
    alphaBoolBlend {blendOption=Linear}
    boneToModify {boneName=dyn_shoulder_flap_b_base_r}
    constraintSetup [
     {targetBone {boneName=upperarm_r}}
     {targetBone {boneName=upperarm_twist_01_r}}
     {targetBone {boneName=upperarm_twist_02_r}}
     {targetBone {boneName=spine_02}}
     {targetBone {boneName=spine_03}}
     {targetBone {boneName=spine_04}}
     {targetBone {boneName=spine_05}}
     {targetBone {boneName=clavicle_r}}
    ]
    constraintWeights [1,1,1,1,1,1,1,1]
   }
   n24 = "Transform (Modify) Bone - Bone: dyn_shoulder_flap_b_base_r"(ComponentPose=@n23, Alpha=1, Translation="0.85,-4,0", Rotation="-5,-3,-7", Scale="1,1,1") {
    rotationMode=BMM_Additive
    rotationSpace=BCS_BoneSpace
    translationMode=BMM_Additive
    translationSpace=BCS_BoneSpace
    alphaBoolBlend {blendOption=Linear}
    alphaScaleBiasClamp {inRange {max=15}}
    boneToModify {boneName=dyn_shoulder_flap_b_base_r}
    rotation=<rot -5 -3 -7>
    translation=<0.85 -4>
   }
   n25 = RigidBody(ComponentPose=@n24) {
    bClampLinearTranslationLimitToRefPose=true
    bOverrideWorldGravity=true
    overridePhysicsAsset=&M_LRG_Choco_Bunny_Physics
    simulationSpace=BaseBoneSpace
    alphaBoolBlend {blendOption=Linear}
    baseBoneRef {boneName=spine_03}
    componentLinearAccScale=<0.5 0.5 0.6>
    componentLinearVelScale=<0.5 0.5 0.6>
    overrideWorldGravity {z=-980}
   }
   n26 = RigidBody(ComponentPose=@n25) {
    bClampLinearTranslationLimitToRefPose=true
    bOverrideWorldGravity=true
    overridePhysicsAsset=&/M_LRG_Choco_Bunny_Ribbon_Physics
    simulationSpace=BaseBoneSpace
    alphaBoolBlend {blendOption=Linear}
    baseBoneRef {boneName=root}
    componentLinearAccScale=<0.3 0.3 0.3>
    componentLinearVelScale=<0.3 0.3 0.3>
    overrideWorldGravity {z=-980}
   }
   n27 = Control Rig(Source=Component To Local(ComponentPose=@n26)) {
    alphaBoolBlend {blendOption=Linear}
    controlRigAssetReference {blueprintRigClass=&/M_LRG_Choco_Bunny_ControlRig.M_LRG_Choco_Bunny_ControlRig_C}
   }
   Output Pose(Result=@n27) {layerGroup=None}
  ]}
 EventGraph {nodes [
   Event BlueprintInitializeAnimation()
   > Set Body_Mesh(Body_Mesh=GetOwnerSkeletalMeshComponent())
   > n1 = Cast To PlayerPawn_Athena(Object=TryGetPawnOwner())
   > n2 = Set Pawn Owner(Pawn Owner:Player Pawn Athena Object Reference=@n1) -> (Output_Get:Player Pawn Athena Object Reference)
   > Set MeshToCopy(MeshToCopy:Skeletal Mesh Component Object Reference=GetSkeletalMeshForPartType(self=@n2, PartType=Head)) -> (Output_Get:Skeletal Mesh Component Object Reference)
   Event BlueprintUpdateAnimation()
   > n3 = Sequence!(execute) -> (then_0, then_1, then_2:Exec, then_3:Exec)
   n4 = GetOwnerSkeletalMeshComponent()
   n5 = GetDeltaTransformFromRefPose(self=@n4, BoneName=thigh_l, BaseName=pelvis)
   Set legUp_L(execute=@n3.then_1, legUp_L:double=BreakRotator(InRot=BreakTransform(InTransform=@n5).Rotation).Yaw) -> (Output_Get:double)
   n6 = BreakRotator(InRot=BreakTransform(InTransform=GetDeltaTransformFromRefPose(self=@n4, BoneName=thigh_r, BaseName=pelvis)).Rotation)
   Set legUp_R(execute=@n3.then_2, legUp_R:double=@n6.Yaw) -> (Output_Get:double)
   > Set legSide_R(legSide_R:double=@n6.Pitch) -> (Output_Get:double)
   n7 = GetDeltaTransformFromRefPose(self=@n4, BoneName=head, BaseName=spine_05)
   Set headFwdBkRot(execute=@n3.then_3, headFwdBkRot:double=BreakRotator(InRot=BreakTransform(InTransform=@n7).Rotation).Yaw) -> (Output_Get:double)
   TryGetPawnOwner()
  ]}
]
variables [
 Body_Mesh {type=SkeletalMeshComponent}
 headFwdBkRot {type=number}
 legSide_R {type=number}
 legUp_L {type=number}
 legUp_R {type=number}
 MeshToCopy {type=SkeletalMeshComponent}
 Pawn Owner {type=PlayerPawn_Athena_C}
]
